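# Post-process style AnimBlueprint for the FeelUnionMoo female body cosmetic, following the parent mesh and adding secondary motion.
# The copied body pose, with face mesh layered from the head, feeds a Control Rig, then necklace, choker and tail constraints.
# Chained AnimDynamics sway necklace, choker and football bones, plus pecs unless bPbBlendEnabled is true.
# Events set LOD morphs and gravity from ShinToll's parameters, and hide tails under capes or long backblings.
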AnimBlueprint /BRCosmetics/Characters/Player/Female/Medium/Bodies/F_MED_FeelUnionMoo/Meshes/F_MED_FeelUnionMoo_AnimBP
target_skeleton=&/F_MED_FeelUnionMoo_Skeleton
graphs [
 AnimGraph {nodes [
   Copy Pose From Mesh() {bUseAttachedParent=true}
   Output Pose(Result=Blend Poses by bool(
     BlendPose_0=Use cached pose 'Dynamics_cache'()
     BlendPose_1=Use cached pose 'post_chest_dyn'()
     BlendTime_0=0
     BlendTime_1=0
     bActiveValue=Get bPbBlendEnabled() -> (bPbBlendEnabled:bool)
    )) {layerGroup=None}
   n1 = Get Face Mesh() -> (Face Mesh:Skeletal Mesh Component Object Reference)
   Save cached pose 'copypose_cache'(Pose=Layered blend per bone(
     BasePose=Copy Pose From Mesh() {bUseAttachedParent=true}
     BlendPoses_0=Copy Pose From Mesh(SourceMeshComponent=@n1) {bUseAttachedParent=true}
    ) {layerSetup [{branchFilters [{boneName=head}]}]}) {cacheName=copypose_cache}
   n2 = Use cached pose 'control_rig_cache'()
   n3 = "Constraint - dyn_necklace_02 (PPPPPPPP)"!(
    ComponentPose="Transform (Modify) Bone - Bone: helper_spine_04"(ComponentPose=Local To Component(LocalPose=@n2), Alpha=1, Translation="0,0,0", Rotation="0,0,0", Scale="1,1,0.8") {scaleMode=BMM_Additive, boneToModify {boneName=helper_spine_04}, scale {z=0.8}}
    Alpha=1
    ConstraintWeights_0=0.6697
    ConstraintWeights_1:float=0.1731
    ConstraintWeights_2:float=0.0785
    ConstraintWeights_3:float=0.0438
    ConstraintWeights_4:float=0.0195
    ConstraintWeights_5:float=0.0052
    ConstraintWeights_6:float=0.0052
    ConstraintWeights_7:float=0.005
   ) -> (Pose) {
    boneToModify {boneName=dyn_necklace_02}
    constraintSetup [
     {targetBone {boneName=spine_05}}
     {targetBone {boneName=spine_04}}
     {targetBone {boneName=neck_01}}
     {targetBone {boneName=neck_02}}
     {targetBone {boneName=spine_03}}
     {targetBone {boneName=clavicle_l}}
     {targetBone {boneName=clavicle_r}}
     {targetBone {boneName=head}}
    ]
    constraintWeights [1,1,1,1,1,1,1,1]
   }
   n4 = "Constraint - dyn_choker_01 (PPPP)"!(
    ComponentPose=@n3
    Alpha=1
    ConstraintWeights_0=0.3008
    ConstraintWeights_1:float=0.2458
    ConstraintWeights_2:float=0.2411
    ConstraintWeights_3:float=0.2123
   ) -> (Pose) {
    boneToModify {boneName=dyn_choker_01}
    constraintSetup [{targetBone {boneName=neck_02}}, {targetBone {boneName=spine_05}}, {targetBone {boneName=head}}, {targetBone {boneName=neck_01}}]
    constraintWeights [1,1,1,1]
   }
   n5 = "Constraint - tail_set (PPPPPPP)"!(
    ComponentPose=@n4
    Alpha=1
    ConstraintWeights_0=0.5065
    ConstraintWeights_1:float=0.1581
    ConstraintWeights_2:float=0.1092
    ConstraintWeights_3:float=0.1017
    ConstraintWeights_4:float=0.0869
    ConstraintWeights_5:float=0.02
    ConstraintWeights_6:float=0.0176
   ) -> (Pose) {
    boneToModify {boneName=tail_set}
    constraintSetup [
     {targetBone {boneName=pelvis}}
     {targetBone {boneName=deform_glute_r}}
     {targetBone {boneName=deform_glute_l}}
     {targetBone {boneName=thigh_twist_01_r}}
     {targetBone {boneName=thigh_twist_01_l}}
     {targetBone {boneName=thigh_l}}
     {targetBone {boneName=thigh_r}}
    ]
    constraintWeights [1,1,1,1,1,1,1]
   }
   Save cached pose 'predyn_cache'(Pose=Component To Local(ComponentPose=@n5)) {cacheName=predyn_cache}
   n6 = Get Sim Space Settings() -> (Sim Space Settings:Anim Phys Sim Space Settings Structure)
   n7 = Anim Dynamics - Bone: dyn_pec_r!(
    ComponentPose=Local To Component(LocalPose=Use cached pose 'Dynamics_cache'())
    Alpha=1
    AngularSpringConstant:float=5000
    SimSpaceSettings:Anim Phys Sim Space Settings Structure=@n6
   ) -> (Pose) {
    angularBiasOverride=1
    angularDampingOverride=0.9999
    angularSpringConstant=5000
    bAngularSpring=true
    bGravityOverrideInSimSpace=true
    bOverrideAngularBias=true
    bOverrideAngularDamping=true
    bUseGravityOverride=true
    linearDampingOverride=0.9999
    linearSpringConstant=100
    simulationSpace=RootRelative
    boundBone {boneName=dyn_pec_r}
    componentAppliedLinearAccClamp=<2 2 2>
    gravityOverride {z=-980}
    physicsBodyDefinitions [{
      boundBone {boneName=dyn_pec_r}
      boxExtents=<21 21 21>
      constraintSetup {angularLimitsMax=<0 1 5>, angularLimitsMin=<0 -5 -5>, angularTarget=<1 0 0>}
      localJointOffset=<2 0 0>
     }]
    relativeSpaceBone {boneName=spine_04}
    simSpaceSettings {maxAngularAcceleration=45, maxAngularVelocity=8, simSpaceAngularAlpha=0.25}
   }
   n8 = Anim Dynamics - Bone: dyn_pec_l!(ComponentPose=@n7, Alpha=1, AngularSpringConstant:float=5000, SimSpaceSettings:Anim Phys Sim Space Settings Structure=@n6) -> (Pose) {
    angularBiasOverride=1
    angularDampingOverride=0.9999
    angularSpringConstant=5000
    bAngularSpring=true
    bGravityOverrideInSimSpace=true
    bOverrideAngularBias=true
    bOverrideAngularDamping=true
    bUseGravityOverride=true
    linearDampingOverride=0.9999
    linearSpringConstant=100
    simulationSpace=RootRelative
    boundBone {boneName=dyn_pec_l}
    componentAppliedLinearAccClamp=<2 2 2>
    gravityOverride {z=-980}
    physicsBodyDefinitions [{
      boundBone {boneName=dyn_pec_l}
      boxExtents=<20 20 20>
      constraintSetup {angularLimitsMax=<0 1 5>, angularLimitsMin=<0 -5 -5>, angularTarget=<1 0 0>}
      localJointOffset=<2 0 0>
     }]
    relativeSpaceBone {boneName=spine_04}
    simSpaceSettings {maxAngularAcceleration=45, maxAngularVelocity=8, simSpaceAngularAlpha=0.25}
   }
   n9 = Anim Dynamics - Bone: dyn_pec_r!(ComponentPose=@n8, Alpha=1, GravityOverride:Vector="0,0,-200", LinearSpringConstant:float=180) -> (Pose) {
    angularBiasOverride=1
    angularDampingOverride=0.99
    angularSpringConstant=10000
    bGravityOverrideInSimSpace=true
    bLinearSpring=true
    bOverrideAngularBias=true
    bOverrideLinearDamping=true
    bUseGravityOverride=true
    linearDampingOverride=1
    linearSpringConstant=180
    simulationSpace=RootRelative
    boundBone {boneName=dyn_pec_r}
    componentAppliedLinearAccClamp=<5 5 5>
    gravityOverride {z=-200}
    physicsBodyDefinitions [{
      boundBone {boneName=dyn_pec_r}
      boxExtents=<80 80 80>
      constraintSetup {angularLimitsMax=<0 0 0>, angularLimitsMin=<0 0 0>, angularTarget=<1 0 0>, linearAxesMax=<0 0.5 1>, linearAxesMin=<0 -0.5 -0.5>}
      localJointOffset=<0.2 0 0>
     }]
    relativeSpaceBone {boneName=spine_04}
    simSpaceSettings {maxAngularAcceleration=50, maxAngularVelocity=10, simSpaceAngularAlpha=0.1}
   }
   n10 = Anim Dynamics - Bone: dyn_pec_l!(ComponentPose=@n9, Alpha=1, GravityOverride:Vector="0,0,-200", LinearSpringConstant:float=180) -> (Pose) {
    angularBiasOverride=1
    angularDampingOverride=0.99
    angularSpringConstant=10000
    bGravityOverrideInSimSpace=true
    bLinearSpring=true
    bOverrideAngularBias=true
    bOverrideLinearDamping=true
    bUseGravityOverride=true
    linearDampingOverride=1
    linearSpringConstant=180
    simulationSpace=RootRelative
    boundBone {boneName=dyn_pec_l}
    componentAppliedLinearAccClamp=<5 5 5>
    gravityOverride {z=-200}
    physicsBodyDefinitions [{
      boundBone {boneName=dyn_pec_l}
      boxExtents=<90 90 90>
      constraintSetup {angularLimitsMax=<0 0 0>, angularLimitsMin=<0 0 0>, angularTarget=<1 0 0>, linearAxesMax=<0 0.5 1>, linearAxesMin=<0 -0.5 -0.5>}
      localJointOffset=<0.2 0 0>
     }]
    relativeSpaceBone {boneName=spine_04}
    simSpaceSettings {maxAngularAcceleration=50, maxAngularVelocity=10, simSpaceAngularAlpha=0.1}
   }
   Save cached pose 'post_chest_dyn'(Pose=Component To Local(ComponentPose=@n10)) {cacheName=post_chest_dyn}
   n11 = Control Rig(Source=Use cached pose 'copypose_cache'()) {controlRigAssetReference {controlRigAsset=&/F_MED_FeelUnionMoo_CtrlRig}}
   Save cached pose 'control_rig_cache'(Pose=@n11) {cacheName=control_rig_cache}
   n12 = Anim Dynamics - Bone: dyn_necklace_02(ComponentPose=Local To Component(LocalPose=Use cached pose 'predyn_cache'()), Alpha=1, ExternalForce="0,0,0") {
    angularDampingOverride=1
    angularSpringConstant=20000
    bAngularSpring=true
    bOverrideAngularDamping=true
    bUseGravityOverride=true
    gravityScale=0.35
    simulationSpace=RootRelative
    boundBone {boneName=dyn_necklace_02}
    componentLinearAccScale=<0.6 0.6 0.6>
    componentLinearVelScale=<0.6 0.6 0.6>
    gravityOverride {z=-980}
    physicsBodyDefinitions [{
      boundBone {boneName=dyn_necklace_02}
      boxExtents=<20 25 25>
      constraintSetup {angularLimitsMax=<0 20 0>, angularLimitsMin=<0 -20 -50>, angularTarget=<1 0 0>}
      localJointOffset=<8 0 0>
     }]
    relativeSpaceBone {boneName=head}
    simSpaceSettings {maxAngularAcceleration=35, maxAngularVelocity=7, simSpaceAngularAlpha=1}
   }
   n13 = Anim Dynamics - Bone: dyn_choker_01(ComponentPose=@n12, Alpha=1, ExternalForce="0,0,0") {
    angularDampingOverride=1
    angularSpringConstant=20000
    bAngularSpring=true
    bOverrideAngularDamping=true
    bUseGravityOverride=true
    gravityScale=0.35
    simulationSpace=RootRelative
    boundBone {boneName=dyn_choker_01}
    componentLinearAccScale=<0.6 0.6 0.6>
    componentLinearVelScale=<0.6 0.6 0.6>
    gravityOverride {z=-980}
    physicsBodyDefinitions [{
      boundBone {boneName=dyn_choker_01}
      boxExtents=<20 25 25>
      constraintSetup {angularLimitsMax=<0 0 15>, angularLimitsMin=<0 -40 -15>, angularTarget=<1 0 0>}
      localJointOffset=<8 0 0>
     }]
    relativeSpaceBone {boneName=head}
    simSpaceSettings {maxAngularAcceleration=35, maxAngularVelocity=7, simSpaceAngularAlpha=1}
   }
   n14 = Anim Dynamics - Bone: base_foot_ball_l!(ComponentPose=@n13, Alpha=1, GravityOverride:Vector="0,0,-980") -> (Pose) {
    angularDampingOverride=1
    angularSpringConstant=120
    bAngularSpring=true
    bOverrideAngularDamping=true
    bUseGravityOverride=true
    gravityScale=0.35
    simulationSpace=RootRelative
    boundBone {boneName=base_foot_ball_l}
    componentLinearAccScale=<1 1 1>
    componentLinearVelScale=<1 1 1>
    gravityOverride {z=-980}
    physicsBodyDefinitions [{
      boundBone {boneName=base_foot_ball_l}
      boxExtents=<15 20 20>
      constraintSetup {angularLimitsMax=<0 20 20>, angularLimitsMin=<0 -20 0>, angularTarget=<1 0 0>}
      localJointOffset=<3 0 0>
     }]
    simSpaceSettings {maxAngularAcceleration=400, maxAngularVelocity=700, simSpaceAngularAlpha=1}
   }
   n15 = Anim Dynamics - Bone: dyn_football_front_l!(ComponentPose=@n14, Alpha=1, GravityOverride:Vector="0,0,-980") -> (Pose) {
    angularDampingOverride=1
    angularSpringConstant=120
    bAngularSpring=true
    bOverrideAngularDamping=true
    bUseGravityOverride=true
    gravityScale=0.35
    simulationSpace=RootRelative
    boundBone {boneName=dyn_football_front_l}
    componentLinearAccScale=<1 1 1>
    componentLinearVelScale=<1 1 1>
    gravityOverride {z=-980}
    physicsBodyDefinitions [{
      boundBone {boneName=dyn_football_front_l}
      boxExtents=<30 30 30>
      constraintSetup {angularLimitsMax=<0 5 0>, angularLimitsMin=<0 -5 0>, angularTarget=<1 0 0>}
      localJointOffset=<1 0 0>
     }]
    simSpaceSettings {maxAngularAcceleration=400, maxAngularVelocity=700, simSpaceAngularAlpha=1}
   }
   n16 = Anim Dynamics - Bone: dyn_football_back_l!(ComponentPose=@n15, Alpha=1, GravityOverride:Vector="0,0,-980") -> (Pose) {
    angularDampingOverride=1
    angularSpringConstant=120
    bAngularSpring=true
    bOverrideAngularDamping=true
    bUseGravityOverride=true
    gravityScale=0.35
    simulationSpace=RootRelative
    boundBone {boneName=dyn_football_back_l}
    componentLinearAccScale=<1 1 1>
    componentLinearVelScale=<1 1 1>
    gravityOverride {z=-980}
    physicsBodyDefinitions [{
      boundBone {boneName=dyn_football_back_l}
      boxExtents=<30 30 30>
      constraintSetup {angularLimitsMax=<0 0 0>, angularLimitsMin=<0 0 0>, angularTarget=<1 0 0>}
      localJointOffset=<1 0 0>
     }]
    simSpaceSettings {maxAngularAcceleration=400, maxAngularVelocity=700, simSpaceAngularAlpha=1}
   }
   n17 = Anim Dynamics - Bone: base_foot_ball_r!(ComponentPose=@n16, Alpha=1, GravityOverride:Vector="0,0,-980") -> (Pose) {
    angularDampingOverride=1
    angularSpringConstant=120
    bAngularSpring=true
    bOverrideAngularDamping=true
    bUseGravityOverride=true
    gravityScale=0.35
    simulationSpace=RootRelative
    boundBone {boneName=base_foot_ball_r}
    componentLinearAccScale=<1 1 1>
    componentLinearVelScale=<1 1 1>
    gravityOverride {z=-980}
    physicsBodyDefinitions [{
      boundBone {boneName=base_foot_ball_r}
      boxExtents=<15 20 20>
      constraintSetup {angularLimitsMax=<0 40 20>, angularLimitsMin=<0 -40 0>, angularTarget=<1 0 0>}
      localJointOffset=<3 0 0>
     }]
    simSpaceSettings {maxAngularAcceleration=400, maxAngularVelocity=700, simSpaceAngularAlpha=1}
   }
   n18 = Anim Dynamics - Bone: dyn_football_front_r!(ComponentPose=@n17, Alpha=1, GravityOverride:Vector="0,0,-980") -> (Pose) {
    angularDampingOverride=1
    angularSpringConstant=120
    bAngularSpring=true
    bOverrideAngularDamping=true
    bUseGravityOverride=true
    gravityScale=0.35
    simulationSpace=RootRelative
    boundBone {boneName=dyn_football_front_r}
    componentLinearAccScale=<1 1 1>
    componentLinearVelScale=<1 1 1>
    gravityOverride {z=-980}
    physicsBodyDefinitions [{
      boundBone {boneName=dyn_football_front_r}
      boxExtents=<30 30 30>
      constraintSetup {angularLimitsMax=<0 5 0>, angularLimitsMin=<0 -5 0>, angularTarget=<1 0 0>}
      localJointOffset=<1 0 0>
     }]
    simSpaceSettings {maxAngularAcceleration=400, maxAngularVelocity=700, simSpaceAngularAlpha=1}
   }
   n19 = Anim Dynamics - Bone: dyn_football_back_r!(ComponentPose=@n18, Alpha=1, GravityOverride:Vector="0,0,-980") -> (Pose) {
    angularDampingOverride=1
    angularSpringConstant=120
    bAngularSpring=true
    bOverrideAngularDamping=true
    bUseGravityOverride=true
    gravityScale=0.35
    simulationSpace=RootRelative
    boundBone {boneName=dyn_football_back_r}
    componentLinearAccScale=<1 1 1>
    componentLinearVelScale=<1 1 1>
    gravityOverride {z=-980}
    physicsBodyDefinitions [{
      boundBone {boneName=dyn_football_back_r}
      boxExtents=<30 30 30>
      constraintSetup {angularLimitsMax=<0 6 0>, angularLimitsMin=<0 -6 0>, angularTarget=<1 0 0>}
      localJointOffset=<1 0 0>
     }]
    simSpaceSettings {maxAngularAcceleration=400, maxAngularVelocity=700, simSpaceAngularAlpha=1}
   }
   Save cached pose 'Dynamics_cache'(Pose=Component To Local(ComponentPose=@n19)) {cacheName=Dynamics_cache}
  ]}
 EventGraph {nodes [
   Event BlueprintUpdateAnimation()
   > n1 = Sequence!(execute) -> (then_0, then_1, then_2:Exec)
   > n2 = "Gravity Override State Function (Parameter Asset) (Nativized)"(
    CharacterPartAnimInstance=Self-Reference()
    GameTimeInSeconds=GetGameTimeInSeconds()
    Parameters=/BRCosmetics/Characters/Player/Female/Medium/Bodies/F_MED_ShinToll/Meshes/F_MED_ShinToll_Parameters
   )
   > Set GravityOverride(GravityOverride=@n2)
   n3 = Get CurrentLOD()
   n4 = Branch(execute=@n1.then_2, Condition="integer > integer"(A=@n3, B=1) {is=title K2Node_CallFunction})
   > SetMorphTarget!(execute, self:Anim Instance Object Reference, MorphTargetName:Name=Body_LOD_JACKET_BS, Value:float=1) -> (then)
   SetMorphTarget(execute=@n4.else, MorphTargetName=Body_LOD_JACKET_BS, Value=0)
   n5 = Branch(execute=@n1.then_1, Condition="To Boolean (Integer)"!(InInt:int=@n3) -> (ReturnValue:bool))
   > SetMorphTarget(MorphTargetName=Body_LOD_BS, Value=0)
   SetMorphTarget(execute=@n5.else, MorphTargetName=Body_LOD_BS, Value=0)
   n6 = Event OnCharacterCustomizationLoaded()
   > n7 = Sequence!(execute) -> (then_0, then_1)
   > n8 = Set Face Mesh(Face Mesh:Skeletal Mesh Component Object Reference=Get Array Item with EFort Custom Part Type(Array=@n6.CustomMeshParts, PartType=Face)) -> (Output_Get:Skeletal Mesh Component Object Reference)
   > AddTickPrerequisiteComponent!(execute, self:Actor Component Object Reference=GetOwningComponent(), PrerequisiteComponent:Actor Component Object Reference=@n8) -> (then)
   n9 = CheckForBackbling()
   Set isWearingBackbling(
    execute=@n7.then_1
    isWearingBackbling=AND Boolean!(A=@n9, B=NOT Boolean(A=@n9.bIsBackblingACape), C:bool=NOT Boolean(A=@n9.bIsBackblingLong)) -> (ReturnValue)
   )
   > Set Is Wearing Cape(Is Wearing Cape=OR Boolean(A=@n9.bIsBackblingACape, B=@n9.bIsBackblingLong))
   > Set TailMesh(TailMesh:Skeletal Mesh Component Object Reference=Get Array Item with EFort Custom Part Type(Array=@n6.CustomMeshParts, PartType=MiscOrTail)) -> (Output_Get:Skeletal Mesh Component Object Reference)
   > n10 = Sequence()
   n11 = CheckForBackbling()
   > Set isWearingCape(isWearingCape=OR Boolean(A=@n11.bIsBackblingACape, B=@n11.bIsBackblingLong))
   > Set characterHasTail(characterHasTail:bool=CheckForTail!(self:Custom Character Part Anim Instance Object Reference) -> (bHasTail:bool)) -> (Output_Get:bool)
   Branch(execute=@n10.then_1, Condition=Get characterHasTail() -> (characterHasTail:bool))
   > n12 = Branch(Condition=Get isWearingCape() -> (isWearingCape:bool))
   n13 = Get TailMesh() -> (TailMesh:Skeletal Mesh Component Object Reference)
   > SetVisibility!(execute, self:Scene Component Object Reference=@n13, bNewVisibility:bool=false, bPropagateToChildren:bool=false) -> (then)
   > Set IsRBANTail(IsRBANTail:bool=false) -> (Output_Get:bool)
   SetVisibility(execute=@n12.else, self=@n13, bNewVisibility=true, bPropagateToChildren=false)
   > Set IsRBANTail(IsRBANTail=true)
  ]}
]
variables [
 characterHasTail {type=boolean}
 Face Mesh {type=SkeletalMeshComponent}
 GravityOverride {type=Vector}
 Is Wearing Cape {type=boolean}
 IsRBANTail {type=boolean}
 isWearingBackbling {type=boolean}
 isWearingCape {type=boolean}
 Local Noise Offset {type=Vector}
 Local Noise Transform Location {type=Vector}
 Mat Instance Dynamic {type=MaterialInstanceDynamic}
 Overlay Mat Instance Dynamic {type=MaterialInstanceDynamic}
 R1 Velocity Transform Location {type=Vector}
 R1 Velocity Transform Velocity {type=Vector}
 R2 Velocity Transform Location {type=Vector}
 R2 Velocity Transform Velocity {type=Vector}
 Sim Space Settings {
  type=AnimPhysSimSpaceSettings
  default {maxAngularAcceleration=45, maxAngularVelocity=8, simSpaceAngularAlpha=0.25, externalAngularVelocity=<0 0 0>}
 }
 SkelMesh {type=SkeletalMeshComponent}
 TailMesh {type=SkeletalMeshComponent}
 Velocity Interp Speed {type=number}
 WPO Pho Turbulence {type=number}
 WPO Pho Wind Intensity {type=number}
]
