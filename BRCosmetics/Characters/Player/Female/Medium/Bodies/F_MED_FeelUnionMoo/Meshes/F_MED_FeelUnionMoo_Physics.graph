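# Physics asset for the F_MED_FeelUnionMoo female medium player body cosmetic, covering only the calves and four dyn_football bones, likely for secondary motion on dangling leg details.
# Each calf has a capsule body, and each front and back football bone on both legs has a small offset sphere.
# Constraints tie each football sphere to its same-side calf with narrow swing limits and zero twist, so any movement stays a slight sway.
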PhysicsAsset /BRCosmetics/Characters/Player/Female/Medium/Bodies/F_MED_FeelUnionMoo/Meshes/F_MED_FeelUnionMoo_Physics
bodies {
 calf_l [{shapeType=Capsule, length=29, radius=8, center=<-24.8949 -2.5816 0.284>, rotation=<rot 82.5715 179.9987 179.9997>}]
 calf_r [{shapeType=Capsule, length=29, radius=8, center=<24.8949 2.5816 -0.284>, rotation=<rot 82.5715 179.9987 -0.0003>}]
 dyn_football_back_l [{shapeType=Sphere, radius=3, shapeName=dyn_football_back_l_sphere, center=<7.0091 0.2002 0.7826>}]
 dyn_football_back_r [{shapeType=Sphere, radius=3, shapeName=dyn_football_back_r_sphere, center=<7 -0.3 -0.5>}]
 dyn_football_front_l [{shapeType=Sphere, radius=3, shapeName=dyn_football_front_l_sphere, center=<7.6 0 0.5>}]
 dyn_football_front_r [{shapeType=Sphere, radius=3, shapeName=dyn_football_front_r_sphere, center=<7.7 -0.3 0.2>}]
}
constraints [
 {bone1Name=dyn_football_back_l, bone2Name=calf_l, swing1LimitDegrees=20, swing2LimitDegrees=15, twistLimitDegrees=0}
 {bone1Name=dyn_football_front_l, bone2Name=calf_l, swing1LimitDegrees=20, swing2LimitDegrees=15, twistLimitDegrees=0}
 {bone1Name=dyn_football_back_r, bone2Name=calf_r, swing1LimitDegrees=20, swing2LimitDegrees=15, twistLimitDegrees=0}
 {bone1Name=dyn_football_front_r, bone2Name=calf_r, swing1LimitDegrees=20, swing2LimitDegrees=15, twistLimitDegrees=0}
]
