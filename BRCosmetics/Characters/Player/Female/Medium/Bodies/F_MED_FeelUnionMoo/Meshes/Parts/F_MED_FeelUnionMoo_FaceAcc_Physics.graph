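# Physics asset for the FaceAcc part of the F_MED_FeelUnionMoo female cosmetic body, covering its hair buns, front strands and side hair.
# Capsule bodies on the head, upper spine, clavicles and upper arms accompany those on the hair bones.
# Each bun is a three-link chain from a base locked to the head, and each front strand is a two-link chain from its base.
# The two-capsule side hair pieces constrain directly to the head with small swing angles.
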PhysicsAsset /BRCosmetics/Characters/Player/Female/Medium/Bodies/F_MED_FeelUnionMoo/Meshes/Parts/F_MED_FeelUnionMoo_FaceAcc_Physics
bodies {
 base_bunhair_l [{shapeType=Capsule, length=0.505, radius=0.505, shapeName=base_bunhair_l_capsule, rotation=<rot 0 -90 0>}]
 base_bunhair_r [{shapeType=Capsule, length=0.505, radius=0.505, shapeName=base_bunhair_r_capsule, rotation=<rot 0 -90 0>}]
 base_hair_front_01_l [{shapeType=Capsule, length=1, radius=1, shapeName=base_hair_front_01_l_capsule, rotation=<rot 0 -90 0>}]
 base_hair_front_01_r [{shapeType=Capsule, length=0.505, radius=0.505, shapeName=base_hair_front_01_r_capsule, rotation=<rot 0 -90 0>}]
 clavicle_l [{shapeType=Capsule, length=12, radius=8, shapeName=clavicle_l_capsule, center=<3.3085 0.6894 -3.6>, rotation=<rot 0 90 90>}]
 clavicle_r [{shapeType=Capsule, length=12, radius=8, shapeName=clavicle_r_capsule, center=<-3.3085 -0.0521 3.6>, rotation=<rot 0 90 -90>}]
 dyn_bunhair_01_l [{shapeType=Capsule, length=10, radius=4.5, shapeName=dyn_bunhair_01_l_capsule, center=<8 -5 0>, rotation=<rot 90 0 0>}]
 dyn_bunhair_01_r [{shapeType=Capsule, length=10, radius=4.5, shapeName=dyn_bunhair_01_r_capsule, center=<8 -5 0>, rotation=<rot 90 0 0>}]
 dyn_bunhair_02_l [{shapeType=Capsule, length=8, radius=3.5, shapeName=dyn_bunhair_02_l_capsule, center=<8 0 0>, rotation=<rot 90 0 0>}]
 dyn_bunhair_02_r [{shapeType=Capsule, length=8, radius=3.5, shapeName=dyn_bunhair_02_r_capsule, center=<7 0.1 0>, rotation=<rot 90 0 0>}]
 dyn_bunhair_03_l [{shapeType=Capsule, length=8, radius=3, shapeName=dyn_bunhair_03_l_capsule, center=<5.8854 -1.7986 0>, rotation=<rot 90 -21 0>}]
 dyn_bunhair_03_r [{
   shapeType=Capsule
   length=8
   radius=3
   shapeName=dyn_bunhair_03_r_capsule
   center=<5.4969 -2.7783 -0.0147>
   rotation=<rot 29.9996 3.0001 4.3873>
  }]
 dyn_hair_02_lf [
  {
   shapeType=Capsule
   length=8
   radius=5
   shapeName=dyn_hair_02_lf_capsule
   center=<11.4725 1.8683 -1.0687>
   rotation=<rot -5.5372 111.3824 88.9942>
  }
  {
   shapeType=Capsule
   length=10
   radius=2.8
   shapeName=dyn_hair_02_lf_capsule
   center=<7.8913 3.5347 -8.8192>
   rotation=<rot 68.0385 179.9998 165.0195>
  }
 ]
 dyn_hair_02_rt [
  {
   shapeType=Capsule
   length=10
   radius=7
   shapeName=dyn_hair_02_rt_capsule
   center=<13.6331 -0.5596 -4>
   rotation=<rot -0.0002 64.7901 89.9995>
  }
  {
   shapeType=Capsule
   length=10
   radius=2.8
   shapeName=dyn_hair_02_lf_capsule
   center=<9.6547 -3.7288 -8.1082>
   rotation=<rot 68.0386 179.9998 -163.8728>
  }
 ]
 dyn_hair_front_01_l [{shapeType=Capsule, length=6.1459, radius=2, shapeName=dyn_hair_front_01_l_capsule, center=<7 0 0>, rotation=<rot 90 0 0>}]
 dyn_hair_front_01_r [{shapeType=Capsule, length=6.1537, radius=1, shapeName=dyn_hair_front_01_r_capsule, center=<8 0 0>, rotation=<rot 90 0 0>}]
 dyn_hair_front_02_l [{shapeType=Capsule, length=10, radius=1.5, shapeName=dyn_hair_front_02_l_capsule, center=<6.1188 2 2>, rotation=<rot 63 -140 -160>}]
 dyn_hair_front_02_r [{shapeType=Capsule, length=10, radius=1, shapeName=dyn_hair_front_02_r_capsule, center=<6.4994 -1.6 3.2>, rotation=<rot 67 167 0>}]
 head [{shapeType=Capsule, length=10, radius=16, center=<0 -2.8988 0>, rotation=<rot 90 0 0>}]
 spine_04 [{shapeType=Capsule, length=12.335, radius=10.7982, shapeName=spine_04_capsule, center=<5.3903 -0.2885 0>}]
 upperarm_l [{shapeType=Capsule, length=27.0626, radius=7.5202, shapeName=upperarm_l_capsule, center=<20.3766 0.4067 -2.2777>, rotation=<rot -90 0 0>}]
 upperarm_r [{shapeType=Capsule, length=27.0988, radius=7.484, shapeName=upperarm_r_capsule, center=<-20.3766 -0.1906 2.2777>, rotation=<rot -90 180 0>}]
}
constraints [
 {bone1Name=dyn_bunhair_02_r, bone2Name=dyn_bunhair_01_r, swing1LimitDegrees=20, swing2LimitDegrees=25, twistLimitDegrees=0}
 {bone1Name=dyn_bunhair_03_r, bone2Name=dyn_bunhair_02_r, swing1LimitDegrees=20, swing2LimitDegrees=25, twistLimitDegrees=0}
 {bone1Name=dyn_bunhair_03_l, bone2Name=dyn_bunhair_02_l, swing1LimitDegrees=20, swing2LimitDegrees=25, twistLimitDegrees=0}
 {bone1Name=dyn_bunhair_02_l, bone2Name=dyn_bunhair_01_l, swing1LimitDegrees=20, swing2LimitDegrees=25, twistLimitDegrees=0}
 {bone1Name=dyn_hair_front_02_r, bone2Name=dyn_hair_front_01_r, swing1LimitDegrees=10, swing2LimitDegrees=20, twistLimitDegrees=0}
 {bone1Name=dyn_hair_front_01_r, bone2Name=base_hair_front_01_r, swing1LimitDegrees=12, swing2LimitDegrees=20, twistLimitDegrees=0}
 {bone1Name=dyn_hair_front_02_l, bone2Name=dyn_hair_front_01_l, swing1LimitDegrees=10, swing2LimitDegrees=20, twistLimitDegrees=0}
 {bone1Name=dyn_hair_front_01_l, bone2Name=base_hair_front_01_l, swing1LimitDegrees=12, swing2LimitDegrees=20, twistLimitDegrees=0}
 {bone1Name=dyn_hair_02_rt, bone2Name=head, swing1LimitDegrees=5, swing2LimitDegrees=15, twistLimitDegrees=0}
 {bone1Name=dyn_hair_02_lf, bone2Name=head, swing1LimitDegrees=5, swing2LimitDegrees=15, twistLimitDegrees=0}
 {
  bone1Name=base_bunhair_r
  bone2Name=head
  swing1LimitDegrees=45
  swing1Motion=Locked
  swing2LimitDegrees=45
  swing2Motion=Locked
  twistLimitDegrees=45
  twistMotion=Locked
 }
 {
  bone1Name=base_bunhair_l
  bone2Name=head
  swing1LimitDegrees=45
  swing1Motion=Locked
  swing2LimitDegrees=45
  swing2Motion=Locked
  twistLimitDegrees=45
  twistMotion=Locked
 }
 {bone1Name=dyn_bunhair_01_r, bone2Name=base_bunhair_r, swing1LimitDegrees=15, swing2LimitDegrees=25, twistLimitDegrees=0}
 {bone1Name=dyn_bunhair_01_l, bone2Name=base_bunhair_l, swing1LimitDegrees=15, swing2LimitDegrees=25, twistLimitDegrees=0}
]
