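# Physics asset for the back hair of the F_MED_FeelUnionMoo female player cosmetic, letting the hair swing against the body.
# Center, left and right chains of four capsules each use angular-limited constraints, with cross-links tying the outer tips to the center tip.
# Head, spine, clavicle and upper-arm capsules provide body collision shapes.
# Two small side-hair capsules and the three chain roots have no constraint to the head.
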PhysicsAsset /BRCosmetics/Characters/Player/Female/Medium/Bodies/F_MED_FeelUnionMoo/Meshes/Parts/F_MED_FeelUnionMoo_FaceAcc_Physics_backhair
bodies {
 clavicle_l [{shapeType=Capsule, length=12, radius=8, shapeName=clavicle_l_capsule, center=<3.3409 0.8436 -3.6176>, rotation=<rot 0 90 93>}]
 clavicle_r [{shapeType=Capsule, length=12, radius=8, shapeName=clavicle_r_capsule, center=<-3.3677 0.1117 4.1303>, rotation=<rot 0 90 93>}]
 dyn_hair_01_cr [{shapeType=Capsule, length=4, radius=2, shapeName=dyn_hair_01_cr_capsule, center=<1 0 0>, rotation=<rot 90 0 0>}]
 dyn_hair_01_lf [{shapeType=Capsule, length=4, radius=2, shapeName=dyn_hair_01_lf_capsule, center=<1 0 0>, rotation=<rot 90 0 0>}]
 dyn_hair_01_rt [{shapeType=Capsule, length=4, radius=2, shapeName=dyn_hair_01_rt_capsule, center=<1 0 0>, rotation=<rot 90 0 0>}]
 dyn_hair_02_cr [{shapeType=Capsule, length=8, radius=5.5, shapeName=dyn_hair_02_cr_capsule, center=<6 0 0>, rotation=<rot 0 0 90>}]
 dyn_hair_02_lf [{shapeType=Capsule, length=8, radius=5.5, shapeName=dyn_hair_02_lf_capsule, center=<6 0 0>, rotation=<rot 0 0 90>}]
 dyn_hair_02_rt [{shapeType=Capsule, length=10, radius=5.5, shapeName=dyn_hair_02_rt_capsule, center=<8 0 0>, rotation=<rot 0 0 90>}]
 dyn_hair_03_cr [{shapeType=Capsule, length=8, radius=4.5, shapeName=dyn_hair_03_cr_capsule, center=<8 0 0>, rotation=<rot 0 0 90>}]
 dyn_hair_03_lf [{shapeType=Capsule, length=10, radius=5, shapeName=dyn_hair_03_lf_capsule, center=<6 0 0>, rotation=<rot 0 0 90>}]
 dyn_hair_03_rt [{shapeType=Capsule, length=10, radius=5, shapeName=dyn_hair_03_rt_capsule, center=<6 0 1.4>, rotation=<rot 0 0 90>}]
 dyn_hair_04_cr [{shapeType=Capsule, length=12, radius=7, shapeName=dyn_hair_04_cr_capsule, center=<10 0 -1.8138>, rotation=<rot 0 0 90>}]
 dyn_hair_04_lf [{shapeType=Capsule, length=10, radius=6, shapeName=dyn_hair_04_lf_capsule, center=<5 0 0>, rotation=<rot 0 0 90>}]
 dyn_hair_04_rt [{shapeType=Capsule, length=13, radius=6, shapeName=dyn_hair_04_rt_capsule, center=<4.5 0 0>, rotation=<rot 0 0 90>}]
 dyn_hair_side_02_lf [{shapeType=Capsule, length=3, radius=1, shapeName=dyn_hair_side_02_lf_capsule, rotation=<rot 90 0 0>}]
 dyn_hair_side_02_rt [{shapeType=Capsule, length=3, radius=1, shapeName=dyn_hair_side_02_rt_capsule, rotation=<rot 90 0 0>}]
 head [{shapeType=Capsule, length=10, radius=16, center=<0 -2.8988 0>, rotation=<rot 90 0 0>}]
 spine_02 [{shapeType=Capsule, length=8.182, radius=14.9513, shapeName=spine_02_capsule, center=<-1.6193 0.858 0>}]
 spine_04 [{shapeType=Capsule, length=12.335, radius=10.7982, shapeName=spine_04_capsule, center=<5.3903 0.6949 0>}]
 upperarm_l [{shapeType=Capsule, length=26.5778, radius=8.005, shapeName=upperarm_l_capsule, center=<20.3766 0.5744 -2.2777>, rotation=<rot -90 0 0>}]
 upperarm_r [{shapeType=Capsule, length=27.4145, radius=7.1684, shapeName=upperarm_r_capsule, center=<-20.3766 -0.1906 2.2777>, rotation=<rot -90 180 0>}]
}
constraints [
 {bone1Name=dyn_hair_04_rt, bone2Name=dyn_hair_03_rt, swing1LimitDegrees=20, swing2LimitDegrees=10, twistLimitDegrees=3}
 {bone1Name=dyn_hair_03_rt, bone2Name=dyn_hair_02_rt, swing1LimitDegrees=20, swing2LimitDegrees=15, twistLimitDegrees=1}
 {bone1Name=dyn_hair_02_rt, bone2Name=dyn_hair_01_rt, swing1LimitDegrees=15, swing2LimitDegrees=40, twistLimitDegrees=1}
 {bone1Name=dyn_hair_04_lf, bone2Name=dyn_hair_03_lf, swing1LimitDegrees=20, swing2LimitDegrees=10, twistLimitDegrees=3}
 {bone1Name=dyn_hair_03_lf, bone2Name=dyn_hair_02_lf, swing1LimitDegrees=20, swing2LimitDegrees=15, twistLimitDegrees=1}
 {bone1Name=dyn_hair_02_lf, bone2Name=dyn_hair_01_lf, swing1LimitDegrees=15, swing2LimitDegrees=40, twistLimitDegrees=1}
 {bone1Name=dyn_hair_04_cr, bone2Name=dyn_hair_03_cr, swing1LimitDegrees=10, swing2LimitDegrees=25, twistLimitDegrees=3}
 {bone1Name=dyn_hair_03_cr, bone2Name=dyn_hair_02_cr, swing1LimitDegrees=8, swing2LimitDegrees=35, twistLimitDegrees=3}
 {bone1Name=dyn_hair_02_cr, bone2Name=dyn_hair_01_cr, swing1LimitDegrees=5, swing2LimitDegrees=40, twistLimitDegrees=0}
 {bone1Name=dyn_hair_04_cr, bone2Name=dyn_hair_04_lf, swing1LimitDegrees=15, swing2LimitDegrees=15, twistLimitDegrees=0}
 {bone1Name=dyn_hair_04_cr, bone2Name=dyn_hair_04_rt, swing1LimitDegrees=15, swing2LimitDegrees=15, twistLimitDegrees=0}
]
