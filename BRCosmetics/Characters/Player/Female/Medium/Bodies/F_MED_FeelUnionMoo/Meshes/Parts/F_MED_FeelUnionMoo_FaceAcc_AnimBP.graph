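# Animation Blueprint for the FeelUnionMoo female face accessory, driving its hair bones on the player character.
# Parent and body-mesh poses feed a skydiving-aware Control Rig, then tactical sprint adds hair rotations.
# Backbling flags choose between weighted bone constraints, offset RigidBody back hair for capes, or plain RigidBody, before a final AnimDynamics and RigidBody pass.
# The event graph updates those flags, gravity settings and a zipline vector the AnimGraph never reads.
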AnimBlueprint /BRCosmetics/Characters/Player/Female/Medium/Bodies/F_MED_FeelUnionMoo/Meshes/Parts/F_MED_FeelUnionMoo_FaceAcc_AnimBP
target_skeleton=&/F_MED_FeelUnionMoo_FaceAcc_Skeleton
graphs [
 AnimGraph {nodes [
   n1 = RigidBody!(ComponentPose=Local To Component(), Alpha, OverrideWorldGravity:Vector="0,0,-980") -> (Pose) {
    bClampLinearTranslationLimitToRefPose=true
    bOverrideWorldGravity=true
    overridePhysicsAsset=&/F_MED_FeelUnionMoo_FaceAcc_Physics
    simulationSpace=BaseBoneSpace
    alphaBoolBlend {blendOption=Linear}
    baseBoneRef {boneName=spine_05}
    overrideWorldGravity {z=-980}
    simSpaceSettings {
     dampingAlpha=0
     maxAngularAcceleration=9
     maxAngularVelocity=4
     maxLinearAcceleration=1500
     maxLinearVelocity=350
     worldAlpha=1
     externalLinearDragV=<1.1 1.1>
    }
   }
   Component To Local(ComponentPose=@n1)
   Output Pose(Result=Use cached pose 'dynamicsCache'()) {layerGroup=None}
   n2 = Layered blend per bone(
    BasePose=Copy Pose From Mesh() {bUseAttachedParent=true}
    BlendPoses_0=Copy Pose From Mesh(SourceMeshComponent=Get Body_Mesh()) {bUseAttachedParent=true}
   ) {layerSetup [{branchFilters [{boneName=clavicle_l}, {boneName=clavicle_r}, {boneName=upperarm_r}, {boneName=upperarm_l}]}]}
   n3 = Control Rig!(Source=@n2, Alpha, Skydiving:bool=Get Isallskydiving() -> (Isallskydiving:bool)) -> (Pose) {controlRigAssetReference {controlRigAsset=&/F_MED_FeelUnionMoo_FaceAcc_CtrlRig}}
   Save cached pose 'controlRigCache'(Pose=@n3) {cacheName=controlRigCache}
   n4 = Anim Dynamics - Bone: dyn_hair_upperhead_cr!(
    ComponentPose=Local To Component(LocalPose=Use cached pose 'backpackCapeCache'())
    Alpha=1
    ComponentLinearAccScale:Vector="1,1,1"
    ComponentLinearVelScale:Vector="1,1,1"
   ) -> (Pose) {
    angularDampingOverride=1
    angularSpringConstant=20000
    bAngularSpring=true
    bGravityOverrideInSimSpace=true
    bOverrideAngularDamping=true
    bUseGravityOverride=true
    gravityScale=0.35
    simulationSpace=RootRelative
    alphaBoolBlend {blendOption=Linear}
    boundBone {boneName=dyn_hair_upperhead_cr}
    componentAppliedLinearAccClamp=<3000 3000 3000>
    componentLinearAccScale=<1 1 1>
    componentLinearVelScale=<1 1 1>
    gravityOverride {z=-980}
    physicsBodyDefinitions [{
      boundBone {boneName=dyn_hair_upperhead_cr}
      boxExtents=<30 30 30>
      constraintSetup {angularLimitsMax=<0 2 20>, angularLimitsMin=<0 -15 -15>, angularTarget=<1 0 0>}
      localJointOffset=<3 0 0>
     }]
    simSpaceSettings {maxAngularAcceleration=35, maxAngularVelocity=7, simSpaceAngularAlpha=0.8}
   }
   n5 = RigidBody(ComponentPose=@n4, OverrideWorldGravity="0,0,-980") {
    bClampLinearTranslationLimitToRefPose=true
    bOverrideWorldGravity=true
    overridePhysicsAsset=&F_MED_FeelUnionMoo_FaceAcc_Physics
    simulationSpace=BaseBoneSpace
    alphaBoolBlend {blendOption=Linear}
    baseBoneRef {boneName=spine_05}
    overrideWorldGravity {z=-980}
    simSpaceSettings {
     dampingAlpha=0
     maxAngularAcceleration=9
     maxAngularVelocity=4
     maxLinearAcceleration=1500
     maxLinearVelocity=350
     worldAlpha=1
     externalLinearDragV=<1.1 1.1>
    }
   }
   n6 = Anim Dynamics - Bone: dyn_hair_tips_01_lf!(ComponentPose=@n5, Alpha=1, ComponentLinearAccScale:Vector="1,1,1", ComponentLinearVelScale:Vector="1,1,1") -> (Pose) {
    angularDampingOverride=1
    angularSpringConstant=20000
    bAngularSpring=true
    bGravityOverrideInSimSpace=true
    bOverrideAngularDamping=true
    bUseGravityOverride=true
    gravityScale=0.35
    simulationSpace=RootRelative
    alphaBoolBlend {blendOption=Linear}
    boundBone {boneName=dyn_hair_tips_01_lf}
    componentAppliedLinearAccClamp=<3000 3000 3000>
    componentLinearAccScale=<1 1 1>
    componentLinearVelScale=<1 1 1>
    gravityOverride {z=-980}
    physicsBodyDefinitions [{
      boundBone {boneName=dyn_hair_tips_01_lf}
      boxExtents=<40 40 40>
      constraintSetup {angularLimitsMax=<0 2 20>, angularLimitsMin=<0 -15 -15>, angularTarget=<1 0 0>}
      localJointOffset=<2 0 0>
     }]
    simSpaceSettings {maxAngularAcceleration=35, maxAngularVelocity=7, simSpaceAngularAlpha=0.8}
   }
   n7 = Anim Dynamics - Bone: dyn_hair_tips_02_lf!(ComponentPose=@n6, Alpha=1, ComponentLinearAccScale:Vector="1,1,1", ComponentLinearVelScale:Vector="1,1,1") -> (Pose) {
    angularDampingOverride=1
    angularSpringConstant=20000
    bAngularSpring=true
    bGravityOverrideInSimSpace=true
    bOverrideAngularDamping=true
    bUseGravityOverride=true
    gravityScale=0.35
    simulationSpace=RootRelative
    alphaBoolBlend {blendOption=Linear}
    boundBone {boneName=dyn_hair_tips_02_lf}
    componentAppliedLinearAccClamp=<3000 3000 3000>
    componentLinearAccScale=<1 1 1>
    componentLinearVelScale=<1 1 1>
    gravityOverride {z=-980}
    physicsBodyDefinitions [{
      boundBone {boneName=dyn_hair_tips_02_lf}
      boxExtents=<40 40 40>
      constraintSetup {angularLimitsMax=<0 2 20>, angularLimitsMin=<0 -15 -15>, angularTarget=<1 0 0>}
      localJointOffset=<2 0 0>
     }]
    simSpaceSettings {maxAngularAcceleration=35, maxAngularVelocity=7, simSpaceAngularAlpha=0.8}
   }
   n8 = Anim Dynamics - Bone: dyn_hair_tips_01_rt!(ComponentPose=@n7, Alpha=1, ComponentLinearAccScale:Vector="1,1,1", ComponentLinearVelScale:Vector="1,1,1") -> (Pose) {
    angularDampingOverride=1
    angularSpringConstant=20000
    bAngularSpring=true
    bGravityOverrideInSimSpace=true
    bOverrideAngularDamping=true
    bUseGravityOverride=true
    gravityScale=0.35
    simulationSpace=RootRelative
    alphaBoolBlend {blendOption=Linear}
    boundBone {boneName=dyn_hair_tips_01_rt}
    componentAppliedLinearAccClamp=<3000 3000 3000>
    componentLinearAccScale=<1 1 1>
    componentLinearVelScale=<1 1 1>
    gravityOverride {z=-980}
    physicsBodyDefinitions [{
      boundBone {boneName=dyn_hair_tips_01_rt}
      boxExtents=<30 30 30>
      constraintSetup {angularLimitsMax=<0 2 20>, angularLimitsMin=<0 -15 -15>, angularTarget=<1 0 0>}
      localJointOffset=<3 0 0>
     }]
    simSpaceSettings {maxAngularAcceleration=35, maxAngularVelocity=7, simSpaceAngularAlpha=0.8}
   }
   n9 = Anim Dynamics - Bone: dyn_hair_tips_02_rt!(ComponentPose=@n8, Alpha=1, ComponentLinearAccScale:Vector="1,1,1", ComponentLinearVelScale:Vector="1,1,1") -> (Pose) {
    angularDampingOverride=1
    angularSpringConstant=20000
    bAngularSpring=true
    bGravityOverrideInSimSpace=true
    bOverrideAngularDamping=true
    bUseGravityOverride=true
    gravityScale=0.35
    simulationSpace=RootRelative
    alphaBoolBlend {blendOption=Linear}
    boundBone {boneName=dyn_hair_tips_02_rt}
    componentAppliedLinearAccClamp=<3000 3000 3000>
    componentLinearAccScale=<1 1 1>
    componentLinearVelScale=<1 1 1>
    gravityOverride {z=-980}
    physicsBodyDefinitions [{
      boundBone {boneName=dyn_hair_tips_02_rt}
      boxExtents=<30 30 30>
      constraintSetup {angularLimitsMax=<0 2 20>, angularLimitsMin=<0 -15 -15>, angularTarget=<1 0 0>}
      localJointOffset=<3 0 0>
     }]
    simSpaceSettings {maxAngularAcceleration=35, maxAngularVelocity=7, simSpaceAngularAlpha=0.8}
   }
   Save cached pose 'dynamicsCache'(Pose=Component To Local(ComponentPose=@n9)) {cacheName=dynamicsCache}
   n10 = Use cached pose 'controlRigCache'()
   n11 = Get TacticalSprint() -> (TacticalSprint:double)
   n12 = "Transform (Modify) Bone - Bone: dyn_hair_01_cr"(
    ComponentPose="Transform (Modify) Bone - Bone: dyn_hair_01_lf"(ComponentPose=Local To Component(LocalPose=@n10), Alpha=@n11, Translation="0,0,0", Rotation="0,0,10", Scale="1,1,1") {rotationMode=BMM_Additive, boneToModify {boneName=dyn_hair_01_lf}, rotation {roll=10}}
    Alpha=@n11
    Translation="0,0,0"
    Rotation="0,0,10"
    Scale="1,1,1"
   ) {rotationMode=BMM_Additive, boneToModify {boneName=dyn_hair_01_cr}, rotation {roll=10}}
   Save cached pose 'PreDynamicsPose'(Pose=Component To Local(ComponentPose="Transform (Modify) Bone - Bone: dyn_hair_01_rt"(ComponentPose=@n12, Alpha=@n11, Translation="0,0,0", Rotation="0,0,10", Scale="1,1,1") {rotationMode=BMM_Additive, boneToModify {boneName=dyn_hair_01_rt}, rotation {roll=10}})) {cacheName=PreDynamicsPose}
   n13 = "Transform (Modify) Bone - Bone: offset_hair_lf"(
    ComponentPose=Local To Component(LocalPose=Use cached pose 'PreDynamicsPose'())
    Alpha=1
    Translation="0,0,0"
    Rotation="0,0,15"
    Scale="1,1,1"
   ) {rotationMode=BMM_Additive, boneToModify {boneName=offset_hair_lf}, rotation {roll=15}}
   n14 = RigidBody!(
    ComponentPose="Transform (Modify) Bone - Bone: offset_hair_rt"(
     ComponentPose="Transform (Modify) Bone - Bone: offset_hair_cr"(ComponentPose=@n13, Alpha=1, Translation="0,0,0", Rotation="0,0,18", Scale="1,1,1") {rotationMode=BMM_Additive, boneToModify {boneName=offset_hair_cr}, rotation {roll=18}}
     Alpha=1
     Translation="0,0,0"
     Rotation="0,0,15"
     Scale="1,1,1"
    ) {rotationMode=BMM_Additive, boneToModify {boneName=offset_hair_rt}, rotation {roll=15}}
    Alpha=0.5
    OverrideWorldGravity="0,0,-980"
    SimSpaceSettings:Sim Space Settings Structure=(WorldAlpha=1,VelocityScaleZ=1,DampingAlpha=0,MaxLinearVelocity=350,MaxAngularVelocity=4,MaxLinearAcceleration=1500,MaxAngularAcceleration=9,ExternalLinearDragV=(X=1.1,Y=1.1,Z=0),ExternalLinearVelocity=(X=0,Y=0,Z=0),ExternalAngularVelocity=(X=0,Y=0,Z=0))
   ) -> (Pose) {
    bClampLinearTranslationLimitToRefPose=true
    bOverrideWorldGravity=true
    overridePhysicsAsset=&/F_MED_FeelUnionMoo_FaceAcc_Physics_backhair
    simulationSpace=BaseBoneSpace
    alphaBoolBlend {blendOption=Linear}
    baseBoneRef {boneName=spine_05}
    componentAppliedLinearAccClamp=<0 0 0>
    overrideWorldGravity {z=-980}
    simSpaceSettings {
     dampingAlpha=0
     maxAngularAcceleration=9
     maxAngularVelocity=4
     maxLinearAcceleration=1500
     maxLinearVelocity=350
     worldAlpha=1
     externalLinearDragV=<1.1 1.1>
    }
   }
   n15 = Component To Local(ComponentPose=@n14)
   n16 = Get Is Wearing Cape() -> (Is Wearing Cape:bool)
   n17 = RigidBody(
    ComponentPose=Local To Component(LocalPose=Use cached pose 'PreDynamicsPose'())
    OverrideWorldGravity="0,0,-980"
    SimSpaceSettings=(WorldAlpha=1,VelocityScaleZ=1,DampingAlpha=0,MaxLinearVelocity=350,MaxAngularVelocity=4,MaxLinearAcceleration=1000,MaxAngularAcceleration=9,ExternalLinearDragV=(X=1.1,Y=1.1,Z=0),ExternalLinearVelocity=(X=0,Y=0,Z=0),ExternalAngularVelocity=(X=0,Y=0,Z=0))
   ) {
    bClampLinearTranslationLimitToRefPose=true
    bOverrideWorldGravity=true
    overridePhysicsAsset=&F_MED_FeelUnionMoo_FaceAcc_Physics_backhair
    simulationSpace=BaseBoneSpace
    alphaBoolBlend {blendOption=Linear}
    baseBoneRef {boneName=spine_05}
    componentAppliedLinearAccClamp=<0 0 0>
    overrideWorldGravity {z=-980}
    simSpaceSettings {
     dampingAlpha=0
     maxAngularAcceleration=9
     maxAngularVelocity=4
     maxLinearAcceleration=1000
     maxLinearVelocity=350
     worldAlpha=1
     externalLinearDragV=<1.1 1.1>
    }
   }
   n18 = Component To Local(ComponentPose=@n17)
   n19 = "Constraint - dyn_hair_03_rt (PPPPPPPP)"!(
    ComponentPose=Local To Component(LocalPose=Use cached pose 'PreDynamicsPose'())
    ConstraintWeights_0=0.3664
    ConstraintWeights_1:float=0.2473
    ConstraintWeights_2:float=0.1665
    ConstraintWeights_3:float=0.1334
    ConstraintWeights_4:float=0.0391
    ConstraintWeights_5:float=0.0365
    ConstraintWeights_6:float=0.0092
    ConstraintWeights_7:float=0.0016
   ) -> (Pose) {
    alpha=0.5
    boneToModify {boneName=dyn_hair_03_rt}
    constraintSetup [
     {targetBone {boneName=clavicle_r}}
     {targetBone {boneName=upperarm_twist_01_r}}
     {targetBone {boneName=deform_upperarm_bk_r}}
     {targetBone {boneName=spine_05}}
     {targetBone {boneName=deform_upperarm_up_r}}
     {targetBone {boneName=spine_04}}
     {targetBone {boneName=deform_upperarm_fr_r}}
     {targetBone {boneName=clavicle_l}}
    ]
    constraintWeights [1,1,1,1,1,1,1,1]
   }
   n20 = "Constraint - dyn_hair_04_rt (PPPPPP)"!(
    ComponentPose=@n19
    ConstraintWeights_0=0.3211
    ConstraintWeights_1:float=0.267
    ConstraintWeights_2:float=0.1403
    ConstraintWeights_3:float=0.1346
    ConstraintWeights_4:float=0.1239
    ConstraintWeights_5:float=0.0131
   ) -> (Pose) {
    alpha=0.5
    boneToModify {boneName=dyn_hair_04_rt}
    constraintSetup [
     {targetBone {boneName=upperarm_twist_01_r}}
     {targetBone {boneName=clavicle_r}}
     {targetBone {boneName=spine_04}}
     {targetBone {boneName=spine_05}}
     {targetBone {boneName=deform_upperarm_bk_r}}
     {targetBone {boneName=deform_upperarm_up_r}}
    ]
    constraintWeights [1,1,1,1,1,1]
   }
   n21 = "Constraint - dyn_hair_04_lf (PPPPPP)"!(
    ComponentPose=@n20
    ConstraintWeights_0=0.3225
    ConstraintWeights_1:float=0.2786
    ConstraintWeights_2:float=0.1421
    ConstraintWeights_3:float=0.1207
    ConstraintWeights_4:float=0.1181
    ConstraintWeights_5:float=0.0181
   ) -> (Pose) {
    alpha=0.5
    boneToModify {boneName=dyn_hair_04_lf}
    constraintSetup [
     {targetBone {boneName=upperarm_twist_01_l}}
     {targetBone {boneName=clavicle_l}}
     {targetBone {boneName=spine_05}}
     {targetBone {boneName=spine_04}}
     {targetBone {boneName=deform_upperarm_bk_l}}
     {targetBone {boneName=spine_02}}
    ]
    constraintWeights [1,1,1,1,1,1]
   }
   n22 = "Constraint - dyn_hair_03_lf (PPPPPPPP)"!(
    ComponentPose=@n21
    ConstraintWeights_0=0.3795
    ConstraintWeights_1:float=0.2336
    ConstraintWeights_2:float=0.1587
    ConstraintWeights_3:float=0.1422
    ConstraintWeights_4:float=0.0383
    ConstraintWeights_5:float=0.0369
    ConstraintWeights_6:float=0.009
    ConstraintWeights_7:float=0.0018
   ) -> (Pose) {
    alpha=0.5
    boneToModify {boneName=dyn_hair_03_lf}
    constraintSetup [
     {targetBone {boneName=clavicle_l}}
     {targetBone {boneName=upperarm_twist_01_l}}
     {targetBone {boneName=deform_upperarm_bk_l}}
     {targetBone {boneName=spine_05}}
     {targetBone {boneName=deform_upperarm_up_l}}
     {targetBone {boneName=spine_04}}
     {targetBone {boneName=deform_upperarm_fr_l}}
     {targetBone {boneName=clavicle_r}}
    ]
    constraintWeights [1,1,1,1,1,1,1,1]
   }
   n23 = "Constraint - dyn_hair_04_cr (PPPPPPPP)"!(
    ComponentPose=@n22
    ConstraintWeights_0=0.2818
    ConstraintWeights_1:float=0.2442
    ConstraintWeights_2:float=0.134
    ConstraintWeights_3:float=0.1324
    ConstraintWeights_4:float=0.1116
    ConstraintWeights_5:float=0.0955
    ConstraintWeights_6:float=0.0003
    ConstraintWeights_7:float=0.0002
   ) -> (Pose) {
    alpha=0.5
    boneToModify {boneName=dyn_hair_04_cr}
    constraintSetup [
     {targetBone {boneName=spine_05}}
     {targetBone {boneName=spine_04}}
     {targetBone {boneName=clavicle_r}}
     {targetBone {boneName=clavicle_l}}
     {targetBone {boneName=pelvis}}
     {targetBone {boneName=spine_02}}
     {targetBone {boneName=spine_03}}
     {targetBone {boneName=upperarm_twist_01_l}}
    ]
    constraintWeights [1,1,1,1,1,1,1,1]
   }
   n24 = "Constraint - dyn_hair_03_cr (PPPPPPPP)"!(
    ComponentPose=@n23
    ConstraintWeights_0=0.5737
    ConstraintWeights_1:float=0.1776
    ConstraintWeights_2:float=0.1585
    ConstraintWeights_3:float=0.0808
    ConstraintWeights_4:float=0.0035
    ConstraintWeights_5:float=0.0023
    ConstraintWeights_6:float=0.0019
    ConstraintWeights_7:float=0.0018
   ) -> (Pose) {
    alpha=0.5
    boneToModify {boneName=dyn_hair_03_cr}
    constraintSetup [
     {targetBone {boneName=spine_05}}
     {targetBone {boneName=clavicle_l}}
     {targetBone {boneName=clavicle_r}}
     {targetBone {boneName=spine_04}}
     {targetBone {boneName=spine_03}}
     {targetBone {boneName=spine_02}}
     {targetBone {boneName=upperarm_twist_01_l}}
     {targetBone {boneName=neck_01}}
    ]
    constraintWeights [1,1,1,1,1,1,1,1]
   }
   Save cached pose 'backpackCapeCache'(Pose=Blend Poses by bool(
     BlendPose_0=Component To Local(ComponentPose=@n24)
     BlendPose_1=Blend Poses by bool(BlendPose_0=@n15, BlendPose_1=@n18, bActiveValue=@n16) {blendType=Linear}
     bActiveValue=Get isWearingBackbling()
    ) {blendType=Linear}) {cacheName=backpackCapeCache}
  ]}
 EventGraph {nodes [
   Event OnCharacterCustomizationLoaded()
   > n1 = Cast To PlayerPawn_Athena(Object=TryGetPawnOwner())
   > Set Body_Mesh(Body_Mesh=GetSkeletalMeshForPartType(self=@n1))
   Event BlueprintUpdateAnimation()
   > n2 = Sequence!(execute) -> (then_0, then_1, then_2:Exec, then_3:Exec)
   > n3 = "Gravity Override State Function (Parameter Asset) (Nativized)"(
    CharacterPartAnimInstance=Self-Reference()
    GameTimeInSeconds=GetGameTimeInSeconds()
    Parameters=&/F_MED_FeelUnionMoo_FaceAcc_Parameters
   )
   > Set GravityOverride(GravityOverride=@n3)
   > Set RBAN Settings Out(RBAN Settings Out=@n3.OutRigidBodyAnimNodeParameters)
   n4 = CheckForBackbling()
   n5 = Set isWearingBackbling(
    execute=@n2.then_1
    isWearingBackbling=AND Boolean!(A=@n4, B=NOT Boolean(A=@n4.bIsBackblingACape), C:bool=NOT Boolean(A=@n4.bIsBackblingLong)) -> (ReturnValue)
   )
   > n6 = Set Is Wearing Cape(Is Wearing Cape=OR Boolean(A=@n4.bIsBackblingACape, B=@n4.bIsBackblingLong))
   > Set is Wearing cape and backpack(is Wearing cape and backpack:bool=OR Boolean(A=@n5, B=@n6)) -> (Output_Get:bool)
   n7 = Branch(execute=@n2.then_2, Condition=Get bIsTacticalSprinting())
   > Set TacticalSprint(TacticalSprint:double=1) -> (Output_Get:double)
   Set TacticalSprint(execute=@n7.else, TacticalSprint=0)
   n8 = Branch(execute=@n2.then_3, Condition=OR Boolean(A=Get bIsSkyDiving(), B=Get bIsSkyDivingFromBus() -> (bIsSkyDivingFromBus:bool)))
   > Set Isallskydiving(Isallskydiving:bool=true) -> (Output_Get:bool)
   Set Isallskydiving(execute=@n8.else, Isallskydiving=false)
   n9 = Event BlueprintProcessAnimInputEvent()
   n10 = IsNamed(self=@n9.AnimInputEvent, InEventName=aie.Player.Zipline.Entered)
   n11 = OR Boolean(A=Get bIsMoving2D(), B=Get bIsTacticalSprinting())
   n12 = Get bIsJumping()
   > n13 = Branch(Condition=OR Boolean(A=@n10, B=AND Boolean(A=@n10, B=@n12, C=@n11)))
   > Set animDyn on zipline(animDyn on zipline:Vector="0, 0, 0") -> (Output_Get:Vector)
   Branch(execute=@n13.else, Condition=IsNamed(self=@n9.AnimInputEvent, InEventName=aie.Player.Zipline.Exited))
   > Set animDyn on zipline(animDyn on zipline="0.5,0.5,0.5")
  ]}
]
variables [
 animDyn on zipline {type=Vector}
 Body_Mesh {type=SkeletalMeshComponent}
 Cape Con Alpha {type=number}
 Cape Crouch Tranform Alpha {type=number}
 Cape Dyn Alpha {type=number}
 GravityOverride {type=Vector}
 Is Wearing Cape {type=boolean}
 is Wearing cape and backpack {type=boolean}
 Isallskydiving {type=boolean}
 isWearingBackbling {type=boolean}
 isWearingBackblingCape {type=boolean}
 RBAN Settings Out {type=FortRigidBodyAnimNodeParameters}
 TacticalSprint {type=number}
]
