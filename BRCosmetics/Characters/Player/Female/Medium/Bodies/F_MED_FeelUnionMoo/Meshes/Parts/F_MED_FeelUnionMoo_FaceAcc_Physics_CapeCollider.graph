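# Physics asset for the face accessory part of the female medium FeelUnionMoo cosmetic body, named as a cape collider.
# It places one capsule on each third and fourth bone of three dynamic hair chains (cr, lf, rt).
# Each fourth bone is constrained to the third bone of its chain with identical narrow swing and twist limits.
# No constraints link the three chains to each other.
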PhysicsAsset /BRCosmetics/Characters/Player/Female/Medium/Bodies/F_MED_FeelUnionMoo/Meshes/Parts/F_MED_FeelUnionMoo_FaceAcc_Physics_CapeCollider
bodies {
 dyn_hair_03_cr [{shapeType=Capsule, length=8, radius=6.9463, shapeName=dyn_hair_03_cr_capsule, center=<5 0 1.4069>, rotation=<rot 0 0 90>}]
 dyn_hair_03_lf [{shapeType=Capsule, length=10, radius=7.9463, shapeName=dyn_hair_03_lf_capsule, center=<6 0 0.3476>, rotation=<rot 0 0 90>}]
 dyn_hair_03_rt [{shapeType=Capsule, length=10, radius=8.4463, shapeName=dyn_hair_03_rt_capsule, center=<6 0 1.7476>, rotation=<rot 0 0 90>}]
 dyn_hair_04_cr [{shapeType=Capsule, length=12, radius=8.9463, shapeName=dyn_hair_04_cr_capsule, center=<9.3989 0 -1.2563>, rotation=<rot 0 0 90>}]
 dyn_hair_04_lf [{shapeType=Capsule, length=10, radius=7.9463, shapeName=dyn_hair_04_lf_capsule, center=<5 0 5.6888>, rotation=<rot 0 0 90>}]
 dyn_hair_04_rt [{shapeType=Capsule, length=13, radius=7.9463, shapeName=dyn_hair_04_rt_capsule, center=<4.5 0 6.3724>, rotation=<rot 0 0 90>}]
}
constraints [
 {bone1Name=dyn_hair_04_rt, bone2Name=dyn_hair_03_rt, swing1LimitDegrees=10, swing2LimitDegrees=15, twistLimitDegrees=3}
 {bone1Name=dyn_hair_04_lf, bone2Name=dyn_hair_03_lf, swing1LimitDegrees=10, swing2LimitDegrees=15, twistLimitDegrees=3}
 {bone1Name=dyn_hair_04_cr, bone2Name=dyn_hair_03_cr, swing1LimitDegrees=10, swing2LimitDegrees=15, twistLimitDegrees=3}
]
